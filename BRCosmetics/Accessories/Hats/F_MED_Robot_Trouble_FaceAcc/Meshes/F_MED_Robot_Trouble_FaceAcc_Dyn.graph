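# Physics asset for the F_MED_Robot_Trouble_FaceAcc cosmetic mesh, a face accessory filed under Hats, with a dynamic beanie bone.
# It has two capsule bodies, a smaller one on the head bone and a larger offset, rotated one on dyn_beanie.
# A single constraint links dyn_beanie to head with limited swing and a tighter twist limit, bounding how far the beanie can move relative to the head.
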PhysicsAsset /BRCosmetics/Accessories/Hats/F_MED_Robot_Trouble_FaceAcc/Meshes/F_MED_Robot_Trouble_FaceAcc_Dyn
bodies {
 dyn_beanie [{shapeType=Capsule, length=8.1478, radius=6.5122, center=<4.1772 -2.019 0>, rotation=<rot 90 90 120.0021>}]
 head [{shapeType=Capsule, length=5, radius=5}]
}
constraints [{bone1Name=dyn_beanie, bone2Name=head, swing1LimitDegrees=45, swing2LimitDegrees=45, twistLimitDegrees=20}]
